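; Level-design gizmo actor that marks a combat engagement range in the world, labelled with its distance band and suited weapon types.
; The construction script switches on the CombatDistanceSize enum, from point blank to very long, picks a mesh size and calls that band's setup function.
; Each setup function sets, scales and positions two text renders, then the size mesh is scaled to match.
; The event graph events are empty.
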(event EventBeginPlay)

(event Collision|EventActorBeginOverlap
  (params (OtherActor :type Actor%20Object%20Reference)))

(event EventTick
  (params (DeltaSeconds :type Float%20%28single-precision%29)))

(fn DistanceCloseSetup
  (Variables|Default|SetCombatDistanceSizeText "NSLOCTEXT(\"[B4F3B6E1F5CE0453E7AB34DB79F8B7BC]\", \"A461A3374A22093CF0F87DB86FE9A2B1\", \"40m - Close\")")
  (Variables|Default|SetCombatDistanceDescriptionText "NSLOCTEXT(\"[B4F3B6E1F5CE0453E7AB34DB79F8B7BC]\", \"11D0275A4D3D5E28F1840CA2313918D9\", \"Shotguns, SMGs\")")
  (Rendering|Components|TextRender|SetXScale (Variables|Default|GetSizeText) 15.0)
  (Rendering|Components|TextRender|SetYScale (Variables|Default|GetSizeText) 15.0)
  (Rendering|Components|TextRender|SetText (Variables|Default|GetSizeText) (Variables|Default|GetCombatDistanceSizeText))
  (Rendering|Components|TextRender|SetXScale (Variables|Default|GetDescriptionText) 15.0)
  (Rendering|Components|TextRender|SetYScale (Variables|Default|GetDescriptionText) 15.0)
  (Rendering|Components|TextRender|SetText (Variables|Default|GetDescriptionText) (Variables|Default|GetCombatDistanceDescriptionText))
  (Transformation|SetRelativeLocation (Variables|Default|GetSizeText) "-230.000000,0.000000,100.500000" false false)
  (Transformation|SetRelativeLocation (Variables|Default|GetDescriptionText) "165.000000,0.000000,100.500000" false false))

(fn DistanceLongSetup
  (Variables|Default|SetCombatDistanceSizeText "NSLOCTEXT(\"[B4F3B6E1F5CE0453E7AB34DB79F8B7BC]\", \"B69144D745126CC47E07658E9A4B718D\", \"160m - Long\")")
  (Variables|Default|SetCombatDistanceDescriptionText "NSLOCTEXT(\"[B4F3B6E1F5CE0453E7AB34DB79F8B7BC]\", \"FC97538D429166439447FBB6D8F6B842\", \"ARs, LMGs, DMRs, Bows\")")
  (Rendering|Components|TextRender|SetXScale (Variables|Default|GetSizeText) 7e+01)
  (Rendering|Components|TextRender|SetYScale (Variables|Default|GetSizeText) 7e+01)
  (Rendering|Components|TextRender|SetText (Variables|Default|GetSizeText) (Variables|Default|GetCombatDistanceSizeText))
  (Rendering|Components|TextRender|SetXScale (Variables|Default|GetDescriptionText) 48.0)
  (Rendering|Components|TextRender|SetYScale (Variables|Default|GetDescriptionText) 48.0)
  (Rendering|Components|TextRender|SetText (Variables|Default|GetDescriptionText) (Variables|Default|GetCombatDistanceDescriptionText))
  (Transformation|SetRelativeLocation (Variables|Default|GetSizeText) "-550.000000,0.000000,100.500000" false false)
  (Transformation|SetRelativeLocation (Variables|Default|GetDescriptionText) "950.000000,0.000000,100.500000" false false))

(fn DistanceMediumSetup
  (Variables|Default|SetCombatDistanceSizeText "NSLOCTEXT(\"[B4F3B6E1F5CE0453E7AB34DB79F8B7BC]\", \"994C8AAF465A00E1083612B951E683E1\", \"90m - Medium\")")
  (Variables|Default|SetCombatDistanceDescriptionText "NSLOCTEXT(\"[B4F3B6E1F5CE0453E7AB34DB79F8B7BC]\", \"BA3C44344AF57186F52E5FAE09E46529\", \"SMGs, Pistols, ARs, Explosive\")")
  (Rendering|Components|TextRender|SetXScale (Variables|Default|GetSizeText) 4e+01)
  (Rendering|Components|TextRender|SetYScale (Variables|Default|GetSizeText) 4e+01)
  (Rendering|Components|TextRender|SetText (Variables|Default|GetSizeText) (Variables|Default|GetCombatDistanceSizeText))
  (Rendering|Components|TextRender|SetXScale (Variables|Default|GetDescriptionText) 26.0)
  (Rendering|Components|TextRender|SetYScale (Variables|Default|GetDescriptionText) 26.0)
  (Rendering|Components|TextRender|SetText (Variables|Default|GetDescriptionText) (Variables|Default|GetCombatDistanceDescriptionText))
  (Transformation|SetRelativeLocation (Variables|Default|GetSizeText) "-230.000000,0.000000,100.500000" false false)
  (Transformation|SetRelativeLocation (Variables|Default|GetDescriptionText) "600.000000,0.000000,100.500000" false false))

(fn DistancePointBlankSetup
  (Variables|Default|SetCombatDistanceSizeText "NSLOCTEXT(\"[B4F3B6E1F5CE0453E7AB34DB79F8B7BC]\", \"D62A076F47324C408AB5F388BDFAF6D4\", \"5m to 10m - Point Blank\")")
  (Variables|Default|SetCombatDistanceDescriptionText "NSLOCTEXT(\"[B4F3B6E1F5CE0453E7AB34DB79F8B7BC]\", \"A2A2794D498F4E9072B39CB9F82337E5\", \"Melee, Shotguns\")")
  (Rendering|Components|TextRender|SetXScale (Variables|Default|GetSizeText) 4.0)
  (Rendering|Components|TextRender|SetYScale (Variables|Default|GetSizeText) 4.0)
  (Rendering|Components|TextRender|SetText (Variables|Default|GetSizeText) (Variables|Default|GetCombatDistanceSizeText))
  (Rendering|Components|TextRender|SetXScale (Variables|Default|GetDescriptionText) 4.0)
  (Rendering|Components|TextRender|SetYScale (Variables|Default|GetDescriptionText) 4.0)
  (Rendering|Components|TextRender|SetText (Variables|Default|GetDescriptionText) (Variables|Default|GetCombatDistanceDescriptionText))
  (Transformation|SetRelativeLocation (Variables|Default|GetSizeText) "-15.000000,0.000000,100.500000" false false)
  (Transformation|SetRelativeLocation (Variables|Default|GetDescriptionText) "100.000000,0.000000,100.500000" false false))

(fn DistanceVeryLongSetup
  (Variables|Default|SetCombatDistanceSizeText "NSLOCTEXT(\"[B4F3B6E1F5CE0453E7AB34DB79F8B7BC]\", \"4387DB2F46CE5A255A91AC868D9EC269\", \"Very Long - Streaming\")")
  (Variables|Default|SetCombatDistanceDescriptionText "NSLOCTEXT(\"[B4F3B6E1F5CE0453E7AB34DB79F8B7BC]\", \"DF52026440E8B96AF46DCBA27F73FF8A\", \"DMRs, Snipers\")")
  (Rendering|Components|TextRender|SetXScale (Variables|Default|GetSizeText) 1e+02)
  (Rendering|Components|TextRender|SetYScale (Variables|Default|GetSizeText) 1e+02)
  (Rendering|Components|TextRender|SetText (Variables|Default|GetSizeText) (Variables|Default|GetCombatDistanceSizeText))
  (Rendering|Components|TextRender|SetXScale (Variables|Default|GetDescriptionText) 1e+02)
  (Rendering|Components|TextRender|SetYScale (Variables|Default|GetDescriptionText) 1e+02)
  (Rendering|Components|TextRender|SetText (Variables|Default|GetDescriptionText) (Variables|Default|GetCombatDistanceDescriptionText))
  (Transformation|SetRelativeLocation (Variables|Default|GetSizeText) "-850.000000,0.000000,100.500000" false false)
  (Transformation|SetRelativeLocation (Variables|Default|GetDescriptionText) "2000.000000,0.000000,100.500000" false false))

(fn ConstructionScript
  (switch Utilities|FlowControl|Switch|SwitchonE_CombatDistanceSize (Variables|Default|GetCombatDistanceSize)
    (:NewEnumerator0
      (Variables|Default|SetCombatDistanceMeshSize "10.000000,10.000000,1.000000")
      (CallFunction|DistancePointBlankSetup))
    (:NewEnumerator1
      (Variables|Default|SetCombatDistanceMeshSize "40.000000,40.000000,2.000000")
      (CallFunction|DistanceCloseSetup))
    (:NewEnumerator2
      (Variables|Default|SetCombatDistanceMeshSize "90.000000,90.000000,2.000000")
      (CallFunction|DistanceMediumSetup))
    (:NewEnumerator3
      (Variables|Default|SetCombatDistanceMeshSize "160.000000,160.000000,2.000000")
      (CallFunction|DistanceLongSetup))
    (:NewEnumerator4
      (Variables|Default|SetCombatDistanceMeshSize "275.000000,275.000000,2.000000")
      (CallFunction|DistanceVeryLongSetup)))
  (Transformation|SetWorldScale3D (Variables|Default|GetSizeMesh) (Variables|Default|GetCombatDistanceMeshSize)))
