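; Placeable testing bomb actor that triggers a dynamic building deconstruction at its location shortly after play begins.
; The construction script sizes a box component from the BombExtents variable.
; On BeginPlay, if BombEnabled is set, it waits a fixed delay and prints the billboard's world location as a debug message.
; It then calls MissionHelpers CreateDynamicBuildingDeconstructor at that location, passing half the box extent.
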(event EventBeginPlay
  (bind _billboard (Variables|Default|GetBillboard))
  (bind _returnvalue (Transformation|GetWorldLocation _billboard))
  (if (Variables|Default|GetBombEnabled)
    (Utilities|FlowControl|Delay 8.0)
    (Development|PrintString (Utilities|String|Append "Testing Bomb Used at " (Utilities|String|ToString(Vector) _returnvalue)) true true "(R=0.000000,G=0.660000,B=1.000000,A=1.000000)" 2e+01)
    (MissionHelpers|CreateDynamicBuildingDeconstructor _returnvalue (/ (Variables|/Script/Engine.BoxComponent|GetBoxExtent (Variables|Default|GetBox)) "2.000000,2.000000,2.000000"))))

(fn ConstructionScript
  (Components|Box|SetBoxExtent (Variables|Default|GetBox) (Variables|Default|GetBombExtents) true))
